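# Additive unlit Niagara mesh-particle material for the BariumDemon cape's sword-intersect effect, producing a flame-edged glow tinted by particle color.
# Two panning flame-pattern samples sum to vary sphere-mask radius along UVs, giving a flickering edge.
# SwizzleUV swaps the flame axes, blends in a second-axis mask and offsets time by UV1.
# The powered mask is scaled by particle alpha, vertex red, opacity strength and axis-aligned Fresnel.
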Material /BRCosmetics/Accessories/FORT_Capes/Cape_M_MED_BariumDemon/FX/M_MED_BariumDemon_CapeSwordIntersect
graph {
 properties {
  bUsedWithNiagaraMeshParticles=true
  blendMode=BLEND_Additive
  pixelDepthOffsetMode=PDOM_Legacy
  shadingModel=MSM_Unlit
  usedShadingModels=Unlit
 }
 nodes [
  n1 = ParticleColor()
  MaterialFunctionCall(Emissive=@n1.RGB) {materialFunction=&MF_FX_EmissiveAdjustments}
  n2 = ComponentMask(None=TextureCoordinate()) {g=true, r=true}
  n3 = Multiply(A=TextureCoordinate(), B=ComponentMask(None=VectorParameter() {parameterName=Tiling, defaultValue=<rgba 1 1 0 1>}.RGB) {g=true, r=true})
  n4 = LinearInterpolate(
   A=@n3
   B=MaterialFunctionCall(XY=@n3) {materialFunction=&Swizzle}.YX
   Alpha=ScalarParameter() {parameterName=SwizzleUV}
  )
  n5 = Multiply(A=@n4, B=Constant2Vector() {g=0.75, r=1})
  n6 = TextureCoordinate() {coordinateIndex=1}
  n7 = Add(A=Time(), B=Multiply(A=ScalarParameter() {parameterName=SwizzleUV}, B=ComponentMask(None=@n6) {r=true}))
  n8 = VectorParameter() {parameterName=LayerSpeeds XY ZW, defaultValue=<rgba 0 2 0 0.75>}
  n9 = AppendVector(A=@n8.R, B=@n8.R)
  n10 = Multiply(A=@n4, B=Constant2Vector() {g=0.35, r=0.7})
  n11 = AppendVector(A=@n8.B, B=@n8.B)
  n12 = Add(
   A=TextureSample(UVs=Panner(Coordinate=@n5, Time=@n7, Speed=@n9) {bFractionalPart=true, speedY=2}) {samplerType=SAMPLERTYPE_Masks, texture=/Game/Global/Textures/Default/Masks_and_Gradients/FlamePattern}.B
   B=TextureSample(UVs=Panner(Coordinate=@n10, Time=@n7, Speed=@n11) {bFractionalPart=true, speedY=0.7}) {samplerType=SAMPLERTYPE_Masks, texture=&FlamePattern}.B
  )
  n13 = ScalarParameter() {defaultValue=0.5, parameterName=FlameOffset}
  n14 = LinearInterpolate(
   A=ScalarParameter() {parameterName=Radius_1}
   B=ScalarParameter() {parameterName=Radius_2}
   Alpha=Saturate(None=Multiply(A=@n12, B=@n13) {constB=0.5})
  )
  n15 = SphereMask(A=ComponentMask(None=@n2) {g=true}, B=ScalarParameter() {defaultValue=1, parameterName=p}, Radius=@n14) {attenuationRadius=1.2, hardnessPercent=0}
  n16 = ComponentMask(None=@n2) {r=true}
  n17 = Constant() {r=1}
  n18 = LinearInterpolate(
   A=@n15
   B=Multiply(A=@n15, B=SphereMask(A=@n16, B=@n17, Radius=@n14) {attenuationRadius=1.2, hardnessPercent=0})
   Alpha=ScalarParameter() {parameterName=SwizzleUV}
  )
  n19 = Multiply(A=Multiply(A=@n1.A, B=Power(Base=@n18) {constExponent=4}), B=VertexColor().R)
  n20 = ScalarParameter() {defaultValue=1, parameterName=Opacity_Str}
  n21 = VectorParameter() {parameterName=Fresnel_Axis, defaultValue=<rgba 0 0 0 1>}
  Multiply(
   A=Saturate(None=Multiply(A=@n19, B=@n20))
   B=MaterialFunctionCall!(Axis=Transform(None=@n21.RGB) {transformSourceType=TRANSFORMSOURCE_Instance}, Exponent=ScalarParameter() {parameterName=FresnelExp}, Normal) {materialFunction=/Game/Effects/Fort_Effects/Materials/_Functions/MF_AxisAlignedFrensnel}.Fresnel
  )
  SphereMask() {attenuationRadius=1.2, hardnessPercent=0}
 ]
}
