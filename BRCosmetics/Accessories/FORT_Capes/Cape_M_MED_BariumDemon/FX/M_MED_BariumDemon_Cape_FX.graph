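# Translucent unlit material for the Barium Demon cape, rendering animated, eroding energy with electric bolts.
# Panned noise, plasma and electricity textures, shaped by edge and cape-end masks, drive erosion opacity and emissive color via screen blend and contrast.
# World position offset adds sine waves and normal-based edge push, alongside a view-angle flatten mask.
# Static switch bIsHologram swaps in MF_Holo attributes, reusing the erosion, offset and emissive detail.
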Material /BRCosmetics/Accessories/FORT_Capes/Cape_M_MED_BariumDemon/FX/M_MED_BariumDemon_Cape_FX
graph {
 properties {
  allowTranslucentCustomDepthWrites=true
  bUseMaterialAttributes=true
  bUsedWithMeshDeformer=true
  bUsedWithSkeletalMesh=true
  blendMode=BLEND_Translucent
  opacityMaskClipValue=0.001
  pixelDepthOffsetMode=PDOM_Legacy
  shadingModel=MSM_Unlit
  usedShadingModels=Unlit
 }
 nodes [
  Add()
  Add()
  Constant()
  Constant()
  Constant() {r=0.5}
  Constant() {r=1}
  n1 = VectorParameter() {parameterName=OverlayDetails, defaultValue=<rgba 0 0 0 1>}
  n2 = VectorParameter() {group=OverlayDetails, parameterName=Overlay_Tiling, defaultValue=<rgba 0 0 0 1>}
  n3 = Multiply(A=TextureCoordinate(), B=ComponentMask(None=@n2.RGB) {g=true, r=true})
  n4 = Multiply(A=NamedRerouteUsage(), B=@n2.A)
  n5 = TextureSampleParameter2D(UVs=Panner(Coordinate=Add(A=@n3, B=@n4)) {speedY=-1}) {
   group=OverlayDetails
   parameterName=T_OverlayDetails
   samplerType=SAMPLERTYPE_LinearGrayscale
   texture=/Game/Effects/Fort_Effects/Emotes/UltraLight/Texture/T_PlasmaCoils
  }
  Multiply(A=@n1.RGB, B=Power(Base=@n5.RGB, Exp=@n1.A))
  n6 = NamedRerouteUsage()
  n7 = Add(A=AppendVector(A=Constant(), B=Multiply(A=@n6) {constB=0.1}), B=TextureCoordinate() {uTiling=4})
  n8 = TextureObject() {samplerType=SAMPLERTYPE_LinearColor, texture=&/T_FX_BariumDemon_Masks}
  n9 = TextureSample(UVs=@n7, Tex=@n8) {samplerType=SAMPLERTYPE_Masks, texture=&T_FX_BariumDemon_Masks}
  NamedRerouteDeclaration(None=Multiply(A=@n9.R, B=Constant() {r=1})) {name=EndsErodeMask, nodeColor=<rgba 1 0 0.9375 1>}
  n10 = TextureSample(UVs=TextureCoordinate() {uTiling=4}, Tex=@n8) {samplerType=SAMPLERTYPE_Masks, texture=&T_FX_BariumDemon_Masks}
  NamedRerouteDeclaration(None=VertexInterpolator(VS=OneMinus(None=@n10.R)).PS) {name=Edges, nodeColor=<rgba 1 0 0.5156 1>}
  n11 = NamedRerouteUsage()
  n12 = Multiply(A=Subtract(A=Multiply(A=@n11) {constB=2}), B=ScalarParameter() {parameterName=Distort_Electricitty})
  n13 = Fmod(A=Add(A=TextureCoordinate() {uTiling=2}, B=@n12), B=Constant() {r=1})
  n14 = Multiply(
   A=@n13
   B=ComponentMask(None=VectorParameter() {group=Electricity, parameterName=elec_scale, defaultValue=<rgba -0.5 -0.1 0 1>}.RGB) {g=true, r=true}
  )
  n15 = ComponentMask(None=Constant3Vector() {constant=<rgba 0.5 0.05 0 1>}) {g=true, r=true}
  n16 = Multiply(
   A=Normalize(VectorInput=Subtract(A=@n13, B=@n15))
   B=ScalarParameter() {desc=How much the pinch point pulls the electricity texture, group=Electricity, parameterName=elec_pinch}
  )
  n17 = TextureSampleParameter2D!(UVs=Panner(Coordinate=Add(A=@n14, B=@n16), Time=NamedRerouteUsage()) {speedY=-1}, Level, Apply View MipBias) {
   automaticViewMipBias=false
   constMipValue=0
   group=Electricity
   mipValueMode=TMVM_MipLevel
   parameterName=T_Electricity
   samplerSource=SSM_Wrap_WorldGroupSettings
   texture=/Game/Accessories/FORT_Capes/Cape_M_MED_BariumDemon/FX/T_FX_BariumDemon_Electricity
  }
  n18 = NamedRerouteUsage()
  n19 = ScalarParameter() {parameterName=mul_elec}
  n20 = Power(Base=ComponentMask(None=TextureCoordinate()) {g=true}, Exp=Constant() {r=3})
  n21 = Multiply(A=Multiply(A=Multiply(A=@n17.R) {constB=4}, B=@n18), B=Power(Base=NamedRerouteUsage(), Exp=LinearInterpolate(B=@n19, Alpha=@n20) {constA=1}))
  NamedRerouteDeclaration(None=@n21) {name=TilingElec, nodeColor=<rgba 1 0 0.4922 1>}
  NamedRerouteDeclaration(None=Multiply(A=Time() {period=1}, B=ScalarParameter() {defaultValue=1, parameterName=Speed})) {name=EffectTime, nodeColor=<rgba 1 0.7266 0 1>}
  NamedRerouteDeclaration(None=ScalarParameter() {parameterName=Equipped_Hotspot}) {name=Equipped, nodeColor=<rgba 0.3828 1 0 1>}
  n22 = TextureCoordinate() {uTiling=2, vTiling=0.75}
  n23 = Constant() {r=1}
  n24 = Power(Base=SphereMask(A=Fmod(A=@n22, B=@n23), B=Constant2Vector() {g=0.125, r=0.5}) {attenuationRadius=0.4, hardnessPercent=0}) {constExponent=4}
  NamedRerouteDeclaration(None=VertexInterpolator(VS=Multiply(A=@n24)).PS) {name=HotspotWhenEquipped, nodeColor=<rgba 1 0 0.5156 1>}
  n25 = LinearInterpolate(
   A=ScalarParameter() {group=CapeEnds, parameterName=CapeEnd_Pos}
   B=ScalarParameter() {group=CapeEnds, parameterName=CapeSpikes_BottomEdge_Lift}
   Alpha=@n9.G
  )
  n26 = ScalarParameter() {group=CapeEnds, parameterName=CapeEnd_W}
  n27 = ComponentMask(None=TextureCoordinate()) {g=true}
  n28 = TextureCoordinate() {uTiling=2}
  n29 = NamedRerouteUsage()
  n30 = Multiply(A=Subtract(A=Multiply(A=@n29) {constB=2}), B=ScalarParameter() {group=CapeEnds, parameterName=CapeEndDistort})
  n31 = TextureSample(UVs=Panner(Coordinate=Add(A=@n28, B=@n30)) {speedY=-1}) {samplerType=SAMPLERTYPE_LinearColor, texture=&/T_FX_BariumDemon_CapeTether}
  n32 = Multiply(A=@n31.G, B=ScalarParameter() {group=CapeEnds, parameterName=CapeSpikes_Size})
  NamedRerouteDeclaration(None=SmoothStep(Min=@n25, Max=Add(A=@n25, B=@n26), Value=Add(A=@n27, B=@n32))) {name=CapeTetheredEnds, nodeColor=<rgba 0 0.6016 1 1>}
  NamedRerouteDeclaration(None=Step(X=NamedRerouteUsage()) {constY=0.025}) {name=CapeClipmask, nodeColor=<rgba 0 0.7187 1 1>}
  NamedRerouteDeclaration(None=@n10.R) {name=RawEdges, nodeColor=<rgba 0.1484 0 1 1>}
  n33 = Normalize(VectorInput=Subtract(A=WorldPosition().XYZ, B=CameraPositionWS()))
  n34 = VertexNormalWS()
  n35 = MaterialFunctionCall(Input=Abs(None=DotProduct(A=@n33, B=@n34)), Input Low=Constant() {r=0.2}, Input High=Constant() {r=0.5}) {materialFunction=&RemapValueRange}
  NamedRerouteDeclaration(None=OneMinus(None=Saturate(None=@n35.Result))) {name=VS_EdgeFlattenMask, nodeColor=<rgba 0.6875 1 0 1>}
  NamedRerouteDeclaration(None=@n10.B) {name=TopEdgeGradient, nodeColor=<rgba 0.2656 1 0 1>}
  n36 = Multiply(A=NamedRerouteUsage(), B=NamedRerouteUsage())
  n37 = Saturate(None=ComponentMask(None=Multiply(A=@n7) {constB=3}) {g=true})
  n38 = Multiply(A=Power(Base=OneMinus(None=@n37)), B=Constant() {r=0.35})
  n39 = Constant() {r=1}
  n40 = Multiply(A=Saturate(None=Add(A=@n36, B=@n38)), B=@n39)
  n41 = TextureCoordinate() {vTiling=0.9}
  n42 = Multiply(A=NamedRerouteUsage()) {constB=0.1}
  n43 = Constant() {r=8}
  n44 = Multiply(A=ComponentMask(None=Add(A=@n41, B=@n42)) {g=true}, B=@n43) {constB=6}
  n45 = Constant() {r=2}
  n46 = Power(Base=Divide(
    A=MaterialFunctionCall!(In=@n44, Ramp Width=@n45) {materialFunction=/Engine/Functions/Engine_MaterialFunctions01/Math/SmoothCeil}.Result
    B=@n43
   )) {constExponent=8}
  n47 = NamedRerouteUsage()
  n48 = Add(A=@n47, B=ScalarParameter() {group=EdgeHighlight, parameterName=Edge_Highlight_pos})
  n49 = TextureCoordinate() {uTiling=2}
  n50 = Multiply(A=NamedRerouteUsage(), B=AppendVector(A=Constant(), B=ScalarParameter() {parameterName=Edges_Curl}))
  n51 = ScalarParameter() {defaultValue=1, parameterName=Vertical_Speed}
  n52 = TextureSampleParameter2D!(
   UVs=Panner(
    Coordinate=Add(A=@n49, B=@n50)
    Time=NamedRerouteUsage()
    Speed=MaterialFunctionCall(Y=@n51) {materialFunction=&MakeFloat2}.Result
   ) {bFractionalPart=true, speedY=-0.8}
   Apply View MipBias
  ) {
   parameterName=T_MainNoise
   samplerType=SAMPLERTYPE_LinearGrayscale
   texture=/Game/Effects/Fort_Effects/Emotes/UltraLight/Texture/T_Tiling_Noise_01_GreyScale
  }
  n53 = SmoothStep(Min=Add(A=@n48, B=ScalarParameter() {defaultValue=0.1, group=EdgeHighlight, parameterName=EdgeHighlight_W}), Max=@n48, Value=@n52.G)
  n54 = Power(Base=ComponentMask(None=TextureCoordinate()) {g=true}) {constExponent=4}
  NamedRerouteDeclaration(None=Add(A=Add(A=@n40, B=@n46), B=Multiply(A=@n53, B=@n54))) {name=BrightAreaMask, nodeColor=<rgba 0 1 0.0156 1>}
  NamedRerouteDeclaration(None=@n52.G) {name=AlphaNoiseMain, nodeColor=<rgba 0.8047 1 0 1>}
  NamedRerouteDeclaration(None=OneMinus(None=Constant())) {name=BottomEdge, nodeColor=<rgba 0 0.8594 1 1>}
  n55 = VectorParameter() {parameterName=Energy_Tiling, defaultValue=<rgba 0 0 0 1>}
  n56 = Multiply(A=TextureCoordinate(), B=ComponentMask(None=@n55.RGB) {g=true, r=true})
  n57 = Multiply(A=NamedRerouteUsage(), B=ScalarParameter() {group=Energy, parameterName=Energy_distortion})
  n58 = NamedRerouteUsage()
  n59 = TextureSampleParameter2D(UVs=Panner(Coordinate=Add(A=@n56, B=@n57), Time=@n58) {speedX=-0.1, speedY=-1.18}) {
   group=Energy
   parameterName=T_EnergyTexture
   samplerType=SAMPLERTYPE_Masks
   texture=/Game/Characters/Player/Male/Medium/Bodies/M_MED_Storm_Soldier/Textures/ElectricityNoise_M
  }
  n60 = Multiply(A=@n56, B=@n55.A)
  n61 = TextureSampleParameter2D(UVs=Panner(Coordinate=Add(A=@n57, B=@n60), Time=@n58) {speedX=0.1, speedY=-1.2}) {
   group=Energy
   parameterName=T_EnergyTexture
   samplerType=SAMPLERTYPE_Masks
   texture=&ElectricityNoise_M
  }
  n62 = ScalarParameter() {defaultValue=1.5, group=Energy, parameterName=EnergyPowerBottom}
  n63 = ScalarParameter() {group=Energy, parameterName=EnergyPowerTop}
  n64 = TextureCoordinate() {vTiling=1.25}
  n65 = Saturate(None=OneMinus(None=ComponentMask(None=@n64) {g=true}))
  NamedRerouteDeclaration(None=Power(Base=Add(A=@n59.G, B=@n61.G), Exp=LinearInterpolate(A=@n62, B=@n63, Alpha=@n65)) {constExponent=1.5}) {name=PlasmaEnergy, nodeColor=<rgba 0.6875 1 0 1>}
  n66 = NamedRerouteUsage()
  n67 = NamedRerouteUsage()
  NamedRerouteDeclaration(None=Add(A=Multiply(A=@n66, B=@n39) {constA=1, constB=0}, B=Multiply(A=@n67) {constB=0.08})) {name=Heat, nodeColor=<rgba 1 0 0.5156 1>}
  n68 = Add(A=NamedRerouteUsage(), B=ScalarParameter() {group=Heat, parameterName=Heat_Highlight_pos})
  n69 = Add(A=@n68, B=ScalarParameter() {defaultValue=0.1, group=Heat, parameterName=Heat_Highlight_W})
  n70 = NamedRerouteUsage()
  NamedRerouteDeclaration(None=Multiply(A=SmoothStep(Min=@n69, Max=@n68, Value=@n70), B=ScalarParameter() {parameterName=Heat_Brightness})) {name=HeatTendrils, nodeColor=<rgba 0 1 0.2969 1>}
  NamedRerouteDeclaration(None=Constant3Vector()) {name=OverlayDetails, nodeColor=<rgba 0 1 0.0156 1>}
  n71 = Panner(Coordinate=Add(A=TextureCoordinate() {uTiling=2, vTiling=1.5}, B=@n12), Time=NamedRerouteUsage()) {speedY=-1.25}
  n72 = Multiply(
   A=Multiply(A=TextureSample(UVs=@n71) {texture=&T_FX_BariumDemon_Electricity}.G) {constB=6}
   B=@n18
  )
  n73 = NamedRerouteUsage()
  NamedRerouteDeclaration(None=Multiply(A=Add(A=@n21, B=@n72), B=Power(Base=@n73) {constExponent=3})) {name=ElectricBolts, nodeColor=<rgba 1 0.8672 0 1>}
  NamedRerouteUsage()
  NamedRerouteUsage()
  PreSkinnedNormal()
  n74 = GetMaterialAttributes(None=MaterialFunctionCall!(
    BaseColor=TextureObject() {texture=/Game/Global/Textures/Default/Blanks/T_FlatGrey}
    Normal=TextureObject() {samplerType=SAMPLERTYPE_Normal, texture=&FlatNormalBlue}
   ) {materialFunction=/Game/UI/Frontend/Lobby/HeroSquad/MF_Holo}.Result) {attributeGetTypes [
    #1
    #2
    #3
    #4
   ]}
  n75 = NamedRerouteUsage()
  n76 = Multiply(A=@n74.Opacity, B=Add(A=Multiply(A=@n75) {constB=0.5}) {constB=0.8})
  n77 = ScalarParameter() {parameterName=Hologram_OpacityLvl}
  n78 = Add(A=@n47, B=ScalarParameter() {group=Erosion, parameterName=Erode_Pos})
  n79 = Add(A=@n78, B=ScalarParameter() {defaultValue=0.1, group=EdgeHighlight, parameterName=Erode_W})
  n80 = Multiply(A=Multiply(A=SmoothStep(Min=@n79, Max=@n78, Value=@n52.G), B=NamedRerouteUsage()), B=NamedRerouteUsage())
  n81 = Multiply(A=Saturate(None=Add(A=@n76, B=@n77)), B=@n80)
  n82 = Time()
  n83 = TextureCoordinate() {uTiling=0.5}
  n84 = ComponentMask(None=@n83) {g=true}
  n85 = Subtract(A=@n82, B=@n84)
  n86 = Add(A=ComponentMask(None=@n83) {r=true}, B=@n82)
  n87 = Multiply(A=Multiply(A=Sine(None=@n85) {period=0.75}, B=Cosine(None=@n86) {period=2}), B=Power(Base=@n84))
  n88 = Constant3Vector() {constant=<rgba 0 1 0 0>}
  n89 = Multiply(
   A=Normalize(VectorInput=Transform(None=@n88) {transformSourceType=TRANSFORMSOURCE_Local})
   B=ScalarParameter() {group=WPO, parameterName=VertexShift}
  )
  n90 = VertexNormalWS()
  n91 = TextureSample() {samplerType=SAMPLERTYPE_Masks, texture=&/T_BariumDemonCape_WPO}
  n92 = Multiply(A=@n91.R, B=Multiply(A=ScalarParameter() {group=WPO, parameterName=VS_Edge_Push}, B=NamedRerouteUsage()))
  n93 = Add(A=Multiply(B=Multiply(A=@n87, B=@n89)) {constA=1}, B=Multiply(A=Multiply(A=@n90, B=@n92), B=@n91.G))
  n94 = Add(A=Add(A=NamedRerouteUsage(), B=NamedRerouteUsage()), B=NamedRerouteUsage())
  n95 = NamedRerouteUsage()
  n96 = MaterialFunctionCall!(
   Blend=Multiply(A=Add(A=@n94, B=@n95), B=VectorParameter() {group=Colors, parameterName=Brights, defaultValue=<rgba 0 0 0 1>}.RGB)
   Base=VectorParameter() {group=Colors, parameterName=BaseColor, defaultValue=<rgba 0 0 0 1>}.RGB
  ) {materialFunction=/Engine/Functions/Engine_MaterialFunctions03/Blends/Blend_Screen}
  n97 = ScalarParameter() {defaultValue=1, parameterName=Contrast}
  n98 = MaterialFunctionCall(Emissive=Add(
    A=NamedRerouteUsage()
    B=MaterialFunctionCall!(Color=@n96.Result, Contrast=@n97) {materialFunction=/Engine/Functions/Engine_MaterialFunctions01/ImageAdjustment/Contrast_Preserve_Color}.Result
   )) {materialFunction=&MF_FX_EmissiveAdjustments}
  n99 = ComponentMask(None=Multiply(A=@n98.Result, B=ScalarParameter() {parameterName=Hologram_Emissive_Addtl})) {g=true}
  n100 = VectorParameter() {parameterName=Glow Color Inner, defaultValue=<rgba 0.1 0.175 1 1>}
  n101 = Multiply(A=Saturate(None=Multiply(A=@n99, B=@n100.A)), B=@n100.RGB)
  n102 = VectorParameter() {
   desc="After adding cape emissive detail and biasing the opacity, the effect may need toning down again to stop being overbright."
   parameterName=Hologram_Post_Bright
   defaultValue=<rgba 1 1 1 1>
  }
  n103 = SetMaterialAttributes!(
   MaterialAttributes=@n74.MaterialAttributes
   Opacity=@n81
   Opacity Mask=@n81
   World Position Offset=Add(A=@n74.World Position Offset, B=@n93)
   Emissive Color=Multiply(A=Add(A=@n74.Emissive Color, B=@n101), B=@n102.RGB)
  ) {attributeSetTypes [
    #1
    #2
    #3
    #4
   ]}
  StaticSwitchParameter(
   True=@n103
   False=MakeMaterialAttributes!(
    BaseColor
    Metallic
    Specular
    Roughness
    Anisotropy
    EmissiveColor=@n98.Result
    Opacity=@n80
    OpacityMask=@n80
    Normal
    Tangent
    WorldPositionOffset=@n93
    SubsurfaceColor
    ClearCoat
    ClearCoatRoughness
    AmbientOcclusion
    Refraction
    CustomizedUVs_0
    CustomizedUVs_1
    CustomizedUVs_2
    CustomizedUVs_3
    CustomizedUVs_4
    CustomizedUVs_5
    CustomizedUVs_6
    CustomizedUVs_7
    PixelDepthOffset
    ShadingModel
    Displacement
   )
  ) {parameterName=bIsHologram}
  TextureSample(UVs=Panner(Coordinate=TextureCoordinate() {uTiling=2}) {speedX=0.1}, Tex=@n8) {samplerType=SAMPLERTYPE_Masks, texture=&T_FX_BariumDemon_Masks}
 ]
}
