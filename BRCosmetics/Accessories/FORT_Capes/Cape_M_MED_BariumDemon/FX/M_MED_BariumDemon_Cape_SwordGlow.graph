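# Unlit Modulate-blend Niagara sprite material for the sword glow effect on the BariumDemon cape.
# A soft radial UV mask multiplied by an inverted, low-exponent Fresnel drives a lerp from white to particle color.
# White is neutral under Modulate, so the sprite tints the scene only inside the mask and fades out without using opacity.
# The lerp result feeds the Emissive input of MF_FX_EmissiveAdjustments.
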Material /BRCosmetics/Accessories/FORT_Capes/Cape_M_MED_BariumDemon/FX/M_MED_BariumDemon_Cape_SwordGlow
graph {
 properties {
  bUsedWithNiagaraSprites=true
  blendMode=BLEND_Modulate
  pixelDepthOffsetMode=PDOM_Legacy
  shadingModel=MSM_Unlit
  translucencyPass=MTP_BeforeDOF
  usedShadingModels=Unlit
 }
 nodes [
  Add()
  MaterialFunctionCall() {materialFunction=&Fresnel_Function}
  n1 = SphereMask(A=TextureCoordinate(), B=Constant2Vector() {g=0.5, r=0.5}) {attenuationRadius=0.5, hardnessPercent=0}
  n2 = OneMinus(None=Fresnel() {baseReflectFraction=0, exponent=0.075})
  MaterialFunctionCall(Emissive=LinearInterpolate(A=Constant3Vector() {constant=<rgba 1 1 1 0>}, B=ParticleColor().RGB, Alpha=Multiply(A=@n1, B=@n2))) {materialFunction=&MF_FX_EmissiveAdjustments}
 ]
}
