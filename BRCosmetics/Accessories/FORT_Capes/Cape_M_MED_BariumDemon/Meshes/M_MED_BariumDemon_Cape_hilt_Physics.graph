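# Physics asset for the BariumDemon cape (hilt mesh) cosmetic, setting up the cape bones for physical motion against the wearer's body.
# Five vertical chains of ten capsules hang from tiny MSTR anchor bodies constrained to root, with limited swing and twist along each chain.
# Angularly free cross-constraints tie neighboring chains at selected rows, so the strands move as one sheet.
# Capsules on root, pelvis, spine, arms and legs approximate the body.
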PhysicsAsset /BRCosmetics/Accessories/FORT_Capes/Cape_M_MED_BariumDemon/Meshes/M_MED_BariumDemon_Cape_hilt_Physics
bodies {
 calf_l [{shapeType=Capsule, length=31.5, radius=12.5, shapeName=calf_r_capsule, center=<-23.3518 2 0>, rotation=<rot 90 0 0>}]
 calf_r [{shapeType=Capsule, length=31.5, radius=12.5, shapeName=calf_r_capsule, center=<23.3518 -2 0>, rotation=<rot 90 0 0>}]
 dyn_cape_01 [{shapeType=Capsule, length=7.5, radius=6.005, shapeName=B, center=<6.5042 -0.0002 2.8483>, rotation=<rot 90 0 179.9999>}]
 dyn_cape_010 [{shapeType=Capsule, length=6.0336, radius=7.4714, shapeName=B, center=<7.9861 -0.0002 2.8483>, rotation=<rot 90 0 179.9999>}]
 dyn_cape_010_l [{shapeType=Capsule, length=5.4024, radius=8.7462, shapeName=B, center=<12.5189 -0.0002 2.8482>, rotation=<rot 90 0.0001 180>}]
 dyn_cape_010_mid_l [{shapeType=Capsule, length=5.6235, radius=7.8815, shapeName=B, center=<7.8293 -0.0002 2.8484>, rotation=<rot 90 0 179.9999>}]
 dyn_cape_010_mid_r [{shapeType=Capsule, length=6.4195, radius=7.8797, shapeName=B, center=<7.7551 -0.0002 2.8483>, rotation=<rot 90 0 179.9999>}]
 dyn_cape_010_r [{shapeType=Capsule, length=5.9036, radius=8.9734, shapeName=B, center=<10.9733 -0.0002 0.7862>, rotation=<rot 90 -0.0001 179.9998>}]
 dyn_cape_01_MSTR [{shapeType=Capsule, length=0.01, radius=0.01}]
 dyn_cape_01_l [{shapeType=Capsule, length=7.5, radius=6.005, shapeName=B, center=<6.5042 -0.0002 2.8483>, rotation=<rot 90 0 179.9999>}]
 dyn_cape_01_l_MSTR [{shapeType=Capsule, length=0.01, radius=0.01}]
 dyn_cape_01_mid_l [{shapeType=Capsule, length=7.5, radius=6.005, shapeName=B, center=<6.5042 -0.0002 2.8483>, rotation=<rot 90 0 179.9999>}]
 dyn_cape_01_mid_l_MSTR [{shapeType=Capsule, length=0.01, radius=0.01}]
 dyn_cape_01_mid_r [{shapeType=Capsule, length=7.5, radius=6.005, shapeName=B, center=<6.5042 -0.0002 2.8483>, rotation=<rot 90 0 179.9999>}]
 dyn_cape_01_mid_r_MSTR [{shapeType=Capsule, length=0.01, radius=0.01}]
 dyn_cape_01_r [{shapeType=Capsule, length=7.5, radius=6.005, shapeName=B, center=<6.5042 -0.0002 2.8483>, rotation=<rot 90 0 179.9999>}]
 dyn_cape_01_r_MSTR [{shapeType=Capsule, length=0.01, radius=0.01}]
 dyn_cape_02 [{shapeType=Capsule, length=7.5, radius=6.005, shapeName=B, center=<6.5042 -0.0002 2.8483>, rotation=<rot 90 0 179.9999>}]
 dyn_cape_02_l [{shapeType=Capsule, length=7.5, radius=6.005, shapeName=B, center=<6.5042 -0.0002 2.8483>, rotation=<rot 90 0 179.9999>}]
 dyn_cape_02_mid_l [{shapeType=Capsule, length=7.5, radius=6.005, shapeName=B, center=<6.5042 -0.0002 2.8483>, rotation=<rot 90 0 179.9999>}]
 dyn_cape_02_mid_r [{shapeType=Capsule, length=7.5, radius=6.005, shapeName=B, center=<6.5042 -0.0002 2.8483>, rotation=<rot 90 0 179.9999>}]
 dyn_cape_02_r [{shapeType=Capsule, length=7.5, radius=6.005, shapeName=B, center=<6.5042 -0.0002 2.8483>, rotation=<rot 90 0 179.9999>}]
 dyn_cape_03 [{shapeType=Capsule, length=7.5, radius=6.005, shapeName=B, center=<6.5042 -0.0002 2.8483>, rotation=<rot 90 0 179.9999>}]
 dyn_cape_03_l [{shapeType=Capsule, length=7.5, radius=6.005, shapeName=B, center=<6.5042 -0.0002 2.8483>, rotation=<rot 90 0 179.9999>}]
 dyn_cape_03_mid_l [{shapeType=Capsule, length=7.5, radius=6.005, shapeName=B, center=<6.5042 -0.0002 2.8483>, rotation=<rot 90 0 179.9999>}]
 dyn_cape_03_mid_r [{shapeType=Capsule, length=7.5, radius=6.005, shapeName=B, center=<6.5042 -0.0002 2.8483>, rotation=<rot 90 0 179.9999>}]
 dyn_cape_03_r [{shapeType=Capsule, length=7.5, radius=6.005, shapeName=B, center=<6.5042 -0.0002 2.8483>, rotation=<rot 90 0 179.9999>}]
 dyn_cape_04 [{shapeType=Capsule, length=7.5, radius=6.005, shapeName=B, center=<6.5042 -0.0002 2.8483>, rotation=<rot 90 0 179.9999>}]
 dyn_cape_04_l [{shapeType=Capsule, length=7.5, radius=6.005, shapeName=B, center=<6.5042 -0.0002 2.8483>, rotation=<rot 90 0 179.9999>}]
 dyn_cape_04_mid_l [{shapeType=Capsule, length=7.5, radius=6.005, shapeName=B, center=<6.5042 -0.0002 2.8483>, rotation=<rot 90 0 179.9999>}]
 dyn_cape_04_mid_r [{shapeType=Capsule, length=7.5, radius=6.005, shapeName=B, center=<6.5042 -0.0002 2.8483>, rotation=<rot 90 0 179.9999>}]
 dyn_cape_04_r [{shapeType=Capsule, length=7.5, radius=6.005, shapeName=B, center=<6.5042 -0.0002 2.8483>, rotation=<rot 90 0 179.9999>}]
 dyn_cape_05 [{shapeType=Capsule, length=7.5, radius=6.005, shapeName=B, center=<6.5042 -0.0002 2.8483>, rotation=<rot 90 0 179.9999>}]
 dyn_cape_05_l [{shapeType=Capsule, length=7.5, radius=6.005, shapeName=B, center=<6.5042 -0.0002 2.8483>, rotation=<rot 90 0 179.9999>}]
 dyn_cape_05_mid_l [{shapeType=Capsule, length=7.5, radius=6.005, shapeName=B, center=<6.5042 -0.0002 2.8483>, rotation=<rot 90 0 179.9999>}]
 dyn_cape_05_mid_r [{shapeType=Capsule, length=7.5, radius=6.005, shapeName=B, center=<6.5042 -0.0002 2.8483>, rotation=<rot 90 0 179.9999>}]
 dyn_cape_05_r [{shapeType=Capsule, length=7.5, radius=6.005, shapeName=B, center=<6.5042 -0.0002 2.8483>, rotation=<rot 90 0 179.9999>}]
 dyn_cape_06 [{shapeType=Capsule, length=7.5, radius=6.005, shapeName=B, center=<6.5042 -0.0002 2.8483>, rotation=<rot 90 0 179.9999>}]
 dyn_cape_06_l [{shapeType=Capsule, length=7.5, radius=6.005, shapeName=B, center=<6.5042 -0.0002 2.8483>, rotation=<rot 90 0 179.9999>}]
 dyn_cape_06_mid_l [{shapeType=Capsule, length=7.5, radius=6.005, shapeName=B, center=<6.5042 -0.0002 2.8483>, rotation=<rot 90 0 179.9999>}]
 dyn_cape_06_mid_r [{shapeType=Capsule, length=7.5, radius=6.005, shapeName=B, center=<6.5042 -0.0002 2.8483>, rotation=<rot 90 0 179.9999>}]
 dyn_cape_06_r [{shapeType=Capsule, length=7.5, radius=6.005, shapeName=B, center=<6.5042 -0.0002 2.8483>, rotation=<rot 90 0 179.9999>}]
 dyn_cape_07 [{shapeType=Capsule, length=7.5, radius=6.005, shapeName=B, center=<6.5042 -0.0002 2.8483>, rotation=<rot 90 0 179.9999>}]
 dyn_cape_07_l [{shapeType=Capsule, length=7.5, radius=6.005, shapeName=B, center=<6.5042 -0.0002 2.8483>, rotation=<rot 90 0 179.9999>}]
 dyn_cape_07_mid_l [{shapeType=Capsule, length=7.5, radius=6.005, shapeName=B, center=<6.5042 -0.0002 2.8483>, rotation=<rot 90 0 179.9999>}]
 dyn_cape_07_mid_r [{shapeType=Capsule, length=7.5, radius=6.005, shapeName=B, center=<6.5042 -0.0002 2.8483>, rotation=<rot 90 0 179.9999>}]
 dyn_cape_07_r [{shapeType=Capsule, length=7.5, radius=6.005, shapeName=B, center=<6.5042 -0.0002 2.8483>, rotation=<rot 90 0 179.9999>}]
 dyn_cape_08 [{shapeType=Capsule, length=7.5, radius=6.005, shapeName=B, center=<6.5042 -0.0002 2.8483>, rotation=<rot 90 0 179.9999>}]
 dyn_cape_08_l [{shapeType=Capsule, length=7.5, radius=6.005, shapeName=B, center=<6.5042 -0.0002 2.8483>, rotation=<rot 90 0 179.9999>}]
 dyn_cape_08_mid_l [{shapeType=Capsule, length=7.5, radius=6.005, shapeName=B, center=<6.5042 -0.0002 2.8483>, rotation=<rot 90 0 179.9999>}]
 dyn_cape_08_mid_r [{shapeType=Capsule, length=7.5, radius=6.005, shapeName=B, center=<6.5042 -0.0002 2.8483>, rotation=<rot 90 0 179.9999>}]
 dyn_cape_08_r [{shapeType=Capsule, length=7.5, radius=6.005, shapeName=B, center=<6.5042 -0.0002 2.8483>, rotation=<rot 90 0 179.9999>}]
 dyn_cape_09 [{shapeType=Capsule, length=6.3507, radius=7.1543, shapeName=B, center=<6.5042 -0.0002 2.8483>, rotation=<rot 90 0 179.9999>}]
 dyn_cape_09_l [{shapeType=Capsule, length=6.3507, radius=7.1543, shapeName=B, center=<6.5042 -0.0002 2.8483>, rotation=<rot 90 0 179.9999>}]
 dyn_cape_09_mid_l [{shapeType=Capsule, length=6.3507, radius=7.1543, shapeName=B, center=<6.5042 -0.0002 2.8483>, rotation=<rot 90 0 179.9999>}]
 dyn_cape_09_mid_r [{shapeType=Capsule, length=6.3507, radius=7.1543, shapeName=B, center=<6.5042 -0.0002 2.8483>, rotation=<rot 90 0 179.9999>}]
 dyn_cape_09_r [{shapeType=Capsule, length=6.3507, radius=7.1543, shapeName=B, center=<6.5042 -0.0002 2.8483>, rotation=<rot 90 0 179.9999>}]
 lowerarm_l [{shapeType=Capsule, length=20.5, radius=4.5, shapeName=lowerarm_l_capsule, center=<11.3312 0 0>, rotation=<rot 90 0 0>}]
 lowerarm_r [{shapeType=Capsule, length=20.5, radius=4.5, shapeName=lowerarm_l_capsule, center=<-11.3312 0 0>, rotation=<rot 90 0 0>}]
 pelvis [{shapeType=Capsule, length=6.9637, radius=12.7565, center=<0 1.1083 0>, rotation=<rot 0 -90 0>}]
 root [{shapeType=Capsule, length=25.875, radius=49.9062, center=<-0.0001 52.2458 4.6268>, rotation=<rot 0 0 5.018>}]
 spine_01 [{shapeType=Capsule, length=6.5613, radius=10.1791, center=<0 2.1812 -0.0001>}]
 spine_02 [{shapeType=Capsule, length=4.4705, radius=9.755, center=<0 1.7189 -0.0001>}]
 spine_03 [{shapeType=Capsule, length=7.1914, radius=8.5382, center=<0 1.1599 -0.0001>}]
 spine_04 [{shapeType=Capsule, length=9.7013, radius=7.968, center=<0.6532 1.8714 0>}]
 spine_05 [{shapeType=Capsule, length=15.8663, radius=8.7969, rotation=<rot 0 -90 0>}]
 thigh_l [{shapeType=Capsule, length=24, radius=11, shapeName=thigh_r_capsule, center=<-20 3.3 0>, rotation=<rot 90 0 0>}]
 thigh_r [{shapeType=Capsule, length=24, radius=11, shapeName=thigh_r_capsule, center=<20 -3.3 0>, rotation=<rot 90 0 0>}]
 upperarm_l [{shapeType=Capsule, length=18.0812, radius=4.1347, center=<10.8056 -0.4936 0.3456>, rotation=<rot 90 0 0>}]
 upperarm_r [{shapeType=Capsule, length=18.0812, radius=4.1347, center=<-10.8056 0.4936 -0.3456>, rotation=<rot 90 0 0>}]
}
constraints [
 {bone1Name=dyn_cape_01_l, bone2Name=dyn_cape_01_l_MSTR, swing1LimitDegrees=10, swing2LimitDegrees=10, twistLimitDegrees=15}
 {bone1Name=dyn_cape_02_l, bone2Name=dyn_cape_01_l, swing1LimitDegrees=45, swing2LimitDegrees=70, twistLimitDegrees=15}
 {bone1Name=dyn_cape_03_l, bone2Name=dyn_cape_02_l, swing1LimitDegrees=50, swing2LimitDegrees=50, twistLimitDegrees=15}
 {bone1Name=dyn_cape_04_l, bone2Name=dyn_cape_03_l, swing1LimitDegrees=50, swing2LimitDegrees=50, twistLimitDegrees=15}
 {bone1Name=dyn_cape_05_l, bone2Name=dyn_cape_04_l, swing1LimitDegrees=50, swing2LimitDegrees=50, twistLimitDegrees=15}
 {bone1Name=dyn_cape_06_l, bone2Name=dyn_cape_05_l, swing1LimitDegrees=50, swing2LimitDegrees=50, twistLimitDegrees=15}
 {bone1Name=dyn_cape_07_l, bone2Name=dyn_cape_06_l, swing1LimitDegrees=50, swing2LimitDegrees=50, twistLimitDegrees=15}
 {bone1Name=dyn_cape_08_l, bone2Name=dyn_cape_07_l, swing1LimitDegrees=45, swing2LimitDegrees=45, twistLimitDegrees=15}
 {bone1Name=dyn_cape_09_l, bone2Name=dyn_cape_08_l, swing1LimitDegrees=45, swing2LimitDegrees=45, twistLimitDegrees=15}
 {bone1Name=dyn_cape_010_l, bone2Name=dyn_cape_09_l, swing1LimitDegrees=45, swing2LimitDegrees=45, twistLimitDegrees=15}
 {bone1Name=dyn_cape_01_r, bone2Name=dyn_cape_01_r_MSTR, swing1LimitDegrees=10, swing2LimitDegrees=10, twistLimitDegrees=15}
 {bone1Name=dyn_cape_02_r, bone2Name=dyn_cape_01_r, swing1LimitDegrees=45, swing2LimitDegrees=70, twistLimitDegrees=15}
 {bone1Name=dyn_cape_03_r, bone2Name=dyn_cape_02_r, swing1LimitDegrees=50, swing2LimitDegrees=50, twistLimitDegrees=15}
 {bone1Name=dyn_cape_04_r, bone2Name=dyn_cape_03_r, swing1LimitDegrees=50, swing2LimitDegrees=50, twistLimitDegrees=15}
 {bone1Name=dyn_cape_05_r, bone2Name=dyn_cape_04_r, swing1LimitDegrees=50, swing2LimitDegrees=50, twistLimitDegrees=15}
 {bone1Name=dyn_cape_06_r, bone2Name=dyn_cape_05_r, swing1LimitDegrees=50, swing2LimitDegrees=50, twistLimitDegrees=15}
 {bone1Name=dyn_cape_07_r, bone2Name=dyn_cape_06_r, swing1LimitDegrees=50, swing2LimitDegrees=50, twistLimitDegrees=15}
 {bone1Name=dyn_cape_08_r, bone2Name=dyn_cape_07_r, swing1LimitDegrees=45, swing2LimitDegrees=45, twistLimitDegrees=15}
 {bone1Name=dyn_cape_09_r, bone2Name=dyn_cape_08_r, swing1LimitDegrees=45, swing2LimitDegrees=45, twistLimitDegrees=15}
 {bone1Name=dyn_cape_010_r, bone2Name=dyn_cape_09_r, swing1LimitDegrees=45, swing2LimitDegrees=45, twistLimitDegrees=15}
 {bone1Name=dyn_cape_01_mid_r, bone2Name=dyn_cape_01_mid_r_MSTR, swing1LimitDegrees=20, swing2LimitDegrees=20, twistLimitDegrees=15}
 {bone1Name=dyn_cape_02_mid_r, bone2Name=dyn_cape_01_mid_r, swing1LimitDegrees=45, swing2LimitDegrees=70, twistLimitDegrees=15}
 {bone1Name=dyn_cape_03_mid_r, bone2Name=dyn_cape_02_mid_r, swing1LimitDegrees=50, swing2LimitDegrees=50, twistLimitDegrees=15}
 {bone1Name=dyn_cape_04_mid_r, bone2Name=dyn_cape_03_mid_r, swing1LimitDegrees=50, swing2LimitDegrees=50, twistLimitDegrees=15}
 {bone1Name=dyn_cape_05_mid_r, bone2Name=dyn_cape_04_mid_r, swing1LimitDegrees=50, swing2LimitDegrees=50, twistLimitDegrees=15}
 {bone1Name=dyn_cape_06_mid_r, bone2Name=dyn_cape_05_mid_r, swing1LimitDegrees=50, swing2LimitDegrees=50, twistLimitDegrees=15}
 {bone1Name=dyn_cape_07_mid_r, bone2Name=dyn_cape_06_mid_r, swing1LimitDegrees=50, swing2LimitDegrees=50, twistLimitDegrees=15}
 {bone1Name=dyn_cape_08_mid_r, bone2Name=dyn_cape_07_mid_r, swing1LimitDegrees=45, swing2LimitDegrees=45, twistLimitDegrees=15}
 {bone1Name=dyn_cape_09_mid_r, bone2Name=dyn_cape_08_mid_r, swing1LimitDegrees=45, swing2LimitDegrees=45, twistLimitDegrees=15}
 {bone1Name=dyn_cape_010_mid_r, bone2Name=dyn_cape_09_mid_r, swing1LimitDegrees=45, swing2LimitDegrees=45, twistLimitDegrees=15}
 {bone1Name=dyn_cape_01, bone2Name=dyn_cape_01_MSTR, swing1LimitDegrees=20, swing2LimitDegrees=20, twistLimitDegrees=15}
 {bone1Name=dyn_cape_02, bone2Name=dyn_cape_01, swing1LimitDegrees=45, swing2LimitDegrees=70, twistLimitDegrees=15}
 {bone1Name=dyn_cape_03, bone2Name=dyn_cape_02, swing1LimitDegrees=50, swing2LimitDegrees=50, twistLimitDegrees=15}
 {bone1Name=dyn_cape_04, bone2Name=dyn_cape_03, swing1LimitDegrees=50, swing2LimitDegrees=50, twistLimitDegrees=15}
 {bone1Name=dyn_cape_05, bone2Name=dyn_cape_04, swing1LimitDegrees=50, swing2LimitDegrees=50, twistLimitDegrees=15}
 {bone1Name=dyn_cape_06, bone2Name=dyn_cape_05, swing1LimitDegrees=50, swing2LimitDegrees=50, twistLimitDegrees=15}
 {bone1Name=dyn_cape_07, bone2Name=dyn_cape_06, swing1LimitDegrees=50, swing2LimitDegrees=50, twistLimitDegrees=15}
 {bone1Name=dyn_cape_08, bone2Name=dyn_cape_07, swing1LimitDegrees=45, swing2LimitDegrees=45, twistLimitDegrees=15}
 {bone1Name=dyn_cape_09, bone2Name=dyn_cape_08, swing1LimitDegrees=45, swing2LimitDegrees=45, twistLimitDegrees=15}
 {bone1Name=dyn_cape_010, bone2Name=dyn_cape_09, swing1LimitDegrees=45, swing2LimitDegrees=45, twistLimitDegrees=15}
 {bone1Name=dyn_cape_01_mid_l, bone2Name=dyn_cape_01_mid_l_MSTR, swing1LimitDegrees=20, swing2LimitDegrees=20, twistLimitDegrees=15}
 {bone1Name=dyn_cape_02_mid_l, bone2Name=dyn_cape_01_mid_l, swing1LimitDegrees=45, swing2LimitDegrees=70, twistLimitDegrees=15}
 {bone1Name=dyn_cape_03_mid_l, bone2Name=dyn_cape_02_mid_l, swing1LimitDegrees=50, swing2LimitDegrees=50, twistLimitDegrees=15}
 {bone1Name=dyn_cape_04_mid_l, bone2Name=dyn_cape_03_mid_l, swing1LimitDegrees=50, swing2LimitDegrees=50, twistLimitDegrees=15}
 {bone1Name=dyn_cape_05_mid_l, bone2Name=dyn_cape_04_mid_l, swing1LimitDegrees=50, swing2LimitDegrees=50, twistLimitDegrees=15}
 {bone1Name=dyn_cape_06_mid_l, bone2Name=dyn_cape_05_mid_l, swing1LimitDegrees=50, swing2LimitDegrees=50, twistLimitDegrees=15}
 {bone1Name=dyn_cape_07_mid_l, bone2Name=dyn_cape_06_mid_l, swing1LimitDegrees=50, swing2LimitDegrees=50, twistLimitDegrees=15}
 {bone1Name=dyn_cape_08_mid_l, bone2Name=dyn_cape_07_mid_l, swing1LimitDegrees=45, swing2LimitDegrees=45, twistLimitDegrees=15}
 {bone1Name=dyn_cape_09_mid_l, bone2Name=dyn_cape_08_mid_l, swing1LimitDegrees=45, swing2LimitDegrees=45, twistLimitDegrees=15}
 {bone1Name=dyn_cape_010_mid_l, bone2Name=dyn_cape_09_mid_l, swing1LimitDegrees=45, swing2LimitDegrees=45, twistLimitDegrees=15}
 {
  bone1Name=dyn_cape_03_mid_l
  bone2Name=dyn_cape_03_l
  swing1LimitDegrees=45
  swing1Motion=Free
  swing2LimitDegrees=45
  swing2Motion=Free
  twistLimitDegrees=45
  twistMotion=Free
 }
 {
  bone1Name=dyn_cape_07_mid_l
  bone2Name=dyn_cape_07_l
  swing1LimitDegrees=45
  swing1Motion=Free
  swing2LimitDegrees=45
  swing2Motion=Free
  twistLimitDegrees=45
  twistMotion=Free
 }
 {
  bone1Name=dyn_cape_09_mid_l
  bone2Name=dyn_cape_09_l
  swing1LimitDegrees=45
  swing1Motion=Free
  swing2LimitDegrees=45
  swing2Motion=Free
  twistLimitDegrees=45
  twistMotion=Free
 }
 {
  bone1Name=dyn_cape_010_mid_l
  bone2Name=dyn_cape_010_l
  swing1LimitDegrees=45
  swing1Motion=Free
  swing2LimitDegrees=45
  swing2Motion=Free
  twistLimitDegrees=45
  twistMotion=Free
 }
 {
  bone1Name=dyn_cape_03_mid_r
  bone2Name=dyn_cape_03_r
  swing1LimitDegrees=45
  swing1Motion=Free
  swing2LimitDegrees=45
  swing2Motion=Free
  twistLimitDegrees=45
  twistMotion=Free
 }
 {
  bone1Name=dyn_cape_07_mid_r
  bone2Name=dyn_cape_07_r
  swing1LimitDegrees=45
  swing1Motion=Free
  swing2LimitDegrees=45
  swing2Motion=Free
  twistLimitDegrees=45
  twistMotion=Free
 }
 {
  bone1Name=dyn_cape_09_mid_r
  bone2Name=dyn_cape_09_r
  swing1LimitDegrees=45
  swing1Motion=Free
  swing2LimitDegrees=45
  swing2Motion=Free
  twistLimitDegrees=45
  twistMotion=Free
 }
 {
  bone1Name=dyn_cape_010_mid_r
  bone2Name=dyn_cape_010_r
  swing1LimitDegrees=45
  swing1Motion=Free
  swing2LimitDegrees=45
  swing2Motion=Free
  twistLimitDegrees=45
  twistMotion=Free
 }
 {
  bone1Name=dyn_cape_03_mid_l
  bone2Name=dyn_cape_03
  swing1LimitDegrees=45
  swing1Motion=Free
  swing2LimitDegrees=45
  swing2Motion=Free
  twistLimitDegrees=45
  twistMotion=Free
 }
 {
  bone1Name=dyn_cape_07_mid_l
  bone2Name=dyn_cape_07
  swing1LimitDegrees=45
  swing1Motion=Free
  swing2LimitDegrees=45
  swing2Motion=Free
  twistLimitDegrees=45
  twistMotion=Free
 }
 {
  bone1Name=dyn_cape_09_mid_l
  bone2Name=dyn_cape_09
  swing1LimitDegrees=45
  swing1Motion=Free
  swing2LimitDegrees=45
  swing2Motion=Free
  twistLimitDegrees=45
  twistMotion=Free
 }
 {
  bone1Name=dyn_cape_010_mid_l
  bone2Name=dyn_cape_010
  swing1LimitDegrees=45
  swing1Motion=Free
  swing2LimitDegrees=45
  swing2Motion=Free
  twistLimitDegrees=45
  twistMotion=Free
 }
 {
  bone1Name=dyn_cape_03_mid_r
  bone2Name=dyn_cape_03
  swing1LimitDegrees=45
  swing1Motion=Free
  swing2LimitDegrees=45
  swing2Motion=Free
  twistLimitDegrees=45
  twistMotion=Free
 }
 {
  bone1Name=dyn_cape_07_mid_r
  bone2Name=dyn_cape_07
  swing1LimitDegrees=45
  swing1Motion=Free
  swing2LimitDegrees=45
  swing2Motion=Free
  twistLimitDegrees=45
  twistMotion=Free
 }
 {
  bone1Name=dyn_cape_09_mid_r
  bone2Name=dyn_cape_09
  swing1LimitDegrees=45
  swing1Motion=Free
  swing2LimitDegrees=45
  swing2Motion=Free
  twistLimitDegrees=45
  twistMotion=Free
 }
 {
  bone1Name=dyn_cape_010_mid_r
  bone2Name=dyn_cape_010
  swing1LimitDegrees=45
  swing1Motion=Free
  swing2LimitDegrees=45
  swing2Motion=Free
  twistLimitDegrees=45
  twistMotion=Free
 }
 {
  bone1Name=dyn_cape_04_r
  bone2Name=dyn_cape_04_mid_r
  swing1LimitDegrees=45
  swing1Motion=Free
  swing2LimitDegrees=45
  swing2Motion=Free
  twistLimitDegrees=45
  twistMotion=Free
 }
 {
  bone1Name=dyn_cape_04_mid_r
  bone2Name=dyn_cape_04
  swing1LimitDegrees=45
  swing1Motion=Free
  swing2LimitDegrees=45
  swing2Motion=Free
  twistLimitDegrees=45
  twistMotion=Free
 }
 {
  bone1Name=dyn_cape_04
  bone2Name=dyn_cape_04_mid_l
  swing1LimitDegrees=45
  swing1Motion=Free
  swing2LimitDegrees=45
  swing2Motion=Free
  twistLimitDegrees=45
  twistMotion=Free
 }
 {
  bone1Name=dyn_cape_04_mid_l
  bone2Name=dyn_cape_04_l
  swing1LimitDegrees=45
  swing1Motion=Free
  swing2LimitDegrees=45
  swing2Motion=Free
  twistLimitDegrees=45
  twistMotion=Free
 }
 {
  bone1Name=dyn_cape_01_MSTR
  bone2Name=root
  swing1LimitDegrees=45
  swing1Motion=Free
  swing2LimitDegrees=45
  swing2Motion=Free
  twistLimitDegrees=45
  twistMotion=Free
 }
 {
  bone1Name=dyn_cape_01_mid_l_MSTR
  bone2Name=root
  swing1LimitDegrees=45
  swing1Motion=Free
  swing2LimitDegrees=45
  swing2Motion=Free
  twistLimitDegrees=45
  twistMotion=Free
 }
 {
  bone1Name=dyn_cape_01_l_MSTR
  bone2Name=root
  swing1LimitDegrees=45
  swing1Motion=Free
  swing2LimitDegrees=45
  swing2Motion=Free
  twistLimitDegrees=45
  twistMotion=Free
 }
 {
  bone1Name=dyn_cape_01_mid_r_MSTR
  bone2Name=root
  swing1LimitDegrees=45
  swing1Motion=Free
  swing2LimitDegrees=45
  swing2Motion=Free
  twistLimitDegrees=45
  twistMotion=Free
 }
 {
  bone1Name=dyn_cape_01_r_MSTR
  bone2Name=root
  swing1LimitDegrees=45
  swing1Motion=Free
  swing2LimitDegrees=45
  swing2Motion=Free
  twistLimitDegrees=45
  twistMotion=Free
 }
 {
  bone1Name=pelvis
  bone2Name=root
  swing1LimitDegrees=45
  swing1Motion=Free
  swing2LimitDegrees=45
  swing2Motion=Free
  twistLimitDegrees=45
  twistMotion=Free
 }
 {
  bone1Name=spine_01
  bone2Name=pelvis
  swing1LimitDegrees=45
  swing1Motion=Free
  swing2LimitDegrees=45
  swing2Motion=Free
  twistLimitDegrees=45
  twistMotion=Free
 }
 {
  bone1Name=spine_02
  bone2Name=spine_01
  swing1LimitDegrees=45
  swing1Motion=Free
  swing2LimitDegrees=45
  swing2Motion=Free
  twistLimitDegrees=45
  twistMotion=Free
 }
 {
  bone1Name=spine_03
  bone2Name=spine_02
  swing1LimitDegrees=45
  swing1Motion=Free
  swing2LimitDegrees=45
  swing2Motion=Free
  twistLimitDegrees=45
  twistMotion=Free
 }
 {
  bone1Name=spine_04
  bone2Name=spine_03
  swing1LimitDegrees=45
  swing1Motion=Free
  swing2LimitDegrees=45
  swing2Motion=Free
  twistLimitDegrees=45
  twistMotion=Free
 }
 {
  bone1Name=spine_05
  bone2Name=spine_04
  swing1LimitDegrees=45
  swing1Motion=Free
  swing2LimitDegrees=45
  swing2Motion=Free
  twistLimitDegrees=45
  twistMotion=Free
 }
 {
  bone1Name=upperarm_l
  bone2Name=spine_05
  swing1LimitDegrees=45
  swing1Motion=Free
  swing2LimitDegrees=45
  swing2Motion=Free
  twistLimitDegrees=45
  twistMotion=Free
 }
 {
  bone1Name=lowerarm_l
  bone2Name=upperarm_l
  swing1LimitDegrees=45
  swing1Motion=Free
  swing2LimitDegrees=45
  swing2Motion=Free
  twistLimitDegrees=45
  twistMotion=Free
 }
 {
  bone1Name=upperarm_r
  bone2Name=spine_05
  swing1LimitDegrees=45
  swing1Motion=Free
  swing2LimitDegrees=45
  swing2Motion=Free
  twistLimitDegrees=45
  twistMotion=Free
 }
 {
  bone1Name=lowerarm_r
  bone2Name=upperarm_r
  swing1LimitDegrees=45
  swing1Motion=Free
  swing2LimitDegrees=45
  swing2Motion=Free
  twistLimitDegrees=45
  twistMotion=Free
 }
 {
  bone1Name=thigh_l
  bone2Name=pelvis
  swing1LimitDegrees=45
  swing1Motion=Free
  swing2LimitDegrees=45
  swing2Motion=Free
  twistLimitDegrees=45
  twistMotion=Free
 }
 {
  bone1Name=calf_l
  bone2Name=thigh_l
  swing1LimitDegrees=45
  swing1Motion=Free
  swing2LimitDegrees=45
  swing2Motion=Free
  twistLimitDegrees=45
  twistMotion=Free
 }
 {
  bone1Name=thigh_r
  bone2Name=pelvis
  swing1LimitDegrees=45
  swing1Motion=Free
  swing2LimitDegrees=45
  swing2Motion=Free
  twistLimitDegrees=45
  twistMotion=Free
 }
 {
  bone1Name=calf_r
  bone2Name=thigh_r
  swing1LimitDegrees=45
  swing1Motion=Free
  swing2LimitDegrees=45
  swing2Motion=Free
  twistLimitDegrees=45
  twistMotion=Free
 }
]
